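; Lager world prediction component for Spires that marks an upcoming spire with a server-spawned minimap icon and a client-spawned cloud.
; When the level turns visible, it triggers the cloud's entrance glow and switches the icon to active.
; It finds the spire playspace by hotfix-gated timer retry or fixed delay, then waits for the spire.
; Spire destruction removes markers and expires the owner on authority; the state-change handler's branch is empty.
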(event EventOnPredictionStarted
  (bind _returnvalue (Components|GetOwner))
  (bind _returnvalue_2 (Transformation|GetActorUpVector _returnvalue))
  (bind (_returnvalue_location _returnvalue_rotation) (Math|Transform|BreakTransform (Transformation|GetActorTransform _returnvalue)))
  (bind _cloudrelativelocation (Variables|Default|GetCloudRelativeLocation))
  (bind _returnvalue_3 (* _returnvalue_2 _cloudrelativelocation))
  (bind _returnvalue_4 (+ _returnvalue_location _returnvalue_3))
  ; Spawn Cloud and Minimap Icon on Prediction start
  (Development|PrintString "Spire Prediction - Started")
  (Utilities|SwitchHasAuthority
    (:Authority
      (CallFunction|SpawnMinimapIcon))
    (:Remote
      (Variables|Default|SetCenterLocation _returnvalue_4)
      (Variables|Default|SetCenterRotation _returnvalue_rotation)
      (bind _returnvalue_5 (Game|SpawnActorfromClass "/SpireSharedAssets/Assets/Blueprints/BP_LowAltitudeClouds_V2_Spire.BP_LowAltitudeClouds_V2_Spire_C" (Math|Transform|MakeTransform _returnvalue_4 _returnvalue_rotation) "Undefined" "MultiplyWithRoot" (Components|GetOwner) 0))
      (Variables|Default|SetPredictionCloud _returnvalue_5))))

(event EventOnLevelVisible
  ; Wait for spire playspace to be available once the level is available
  (Class|BPLowAltitudeCloudsV2Spire|SpireEntranceGlow (Variables|Default|GetPredictionCloud)) ; Spire Spawn in Material Glow
  (CallFunction|SetMinimapIcon (Variables|Minimap|GetMinimapIcon_Active))
  (CallFunction|SetMapLegendIconText (Variables|Default|GetMapIconText_Active))
  (Development|PrintString "Spire Prediction - Ended")
  (if (ScalableFloat|GetValueasBool (Variables|Default|GetHF_ShouldKeepTryingToFindSpiresPlayspace))
    (CallFunction|WaitUntilPlayspaceIsAvailable)
    (else
      (Utilities|FlowControl|Delay 2.0) ; Using a delay to wait for the playspace on the \nclient until we figure out an issue in the better \nwait for playspace node there is.
      (bind _returnvalue (SpatialGameplay|GetPlayspaceforLocation self (Transformation|GetActorLocation (Components|GetOwner))))
      (bind _asbp_spire_playspace (Utilities|Casting|CastToBP_SpirePlayspace _returnvalue))
      (CallFunction|OnSpirePlayspaceAvailable _asbp_spire_playspace))))

(event Custom|OnSpirePlayspaceAvailable
  (params (SpirePlayspace :type BP%20Spire%20Playspace%20Object%20Reference))
  ; Wait for the spire to be available
  (bind _output_get (Variables|Default|SetSpirePlayspace SpirePlayspace))
  (bind _returnvalue (Class|SpirePlayspace|GetSpire _output_get))
  (Utilities|IsValid _returnvalue
    (:Is%20Valid
      (CallFunction|OnSpireAvailable))
    (:Is%20Not%20Valid
      (Playspace|BindEventtoNotifyBoundGameplayVolumeSet SpirePlayspace (EventDispatchers|CreateEvent :!event Custom|OnSpireGameplayVolumeSet)))))

(event EventEndPlay
  (params (EndPlayReason :type EEndPlayReason%20Enum :default "Destroyed"))
  (Class|WaitforDiscoverablePlayspacewithFilterAction|Cancel (Variables|Default|GetWaitForPlayspaceTaskRef)))

(event Custom|OnSpireStateChanged
  (params (PreviousState :type Gameplay%20Tag%20Structure%20%28by%20ref%29) (NewState :type Gameplay%20Tag%20Structure%20%28by%20ref%29))
  ; When Spire Activates, Remove Icon
  (if (GameplayTags|MatchesTag NewState "(TagName=\"Spire.Major.State.Active\")")
    _))

(event Custom|OnSpireGameplayVolumeSet
  (params (Playspace :type Playspace%20Object%20Reference))
  ; Wait for the spire to be available
  (CallFunction|OnSpireAvailable))

(event Custom|OnSpireAvailable
  ; Bind to Spire State Changes
  (bind _returnvalue (Class|SpirePlayspace|GetSpire (Variables|Default|GetSpirePlayspace)))
  (Utilities|IsValid (Actor|GetComponentbyClass _returnvalue "/SpireBoss/Gameplay/Components/BP_MajorSpireGameplayComponent.BP_MajorSpireGameplayComponent_C")
    (:Is%20Valid
      (EventDispatchers|AssignOnStateChangedEvent (Actor|GetComponentbyClass _returnvalue "/SpireBoss/Gameplay/Components/BP_MajorSpireGameplayComponent.BP_MajorSpireGameplayComponent_C") (EventDispatchers|CreateEvent :!event Custom|OnSpireStateChanged))
      (Game|AssignOnDestroyed _returnvalue (EventDispatchers|CreateEvent :!event Custom|OnDestroyed)))))

(event Custom|OnDestroyed
  (params (DestroyedActor :type Actor%20Object%20Reference))
  (bind _predictioncloud (Variables|Default|GetPredictionCloud))
  ; When the spire dies, we remove markers, cloud and the level instance itself
  (Class|BPLowAltitudeCloudsV2Spire|FadeOutCloud _predictioncloud)
  (CallFunction|RemoveMinimapIcon)
  (Actor|SetLifeSpan _predictioncloud 5.0)
  (Fortnite|FortShippingLog self "OnSpireDestroyed")
  (Utilities|Time|ClearandInvalidateTimerbyHandle (Variables|Default|GetTryGetPlayspaceTimerHandle))
  (Utilities|SwitchHasAuthority
    (:Authority
      (Actor|SetLifeSpan (Components|GetOwner) 3.0) ; Let this actor have a chance to call OnDestroyed\n in the client before destroy
    )))

(fn RemoveMinimapIcon
  (Actor|DestroyActor (Variables|Minimap|GetMapIconActor)))

(fn SetMapLegendIconText
  (params (New%20Text :type Text))
  (bind _mapiconactor (Variables|Minimap|GetMapIconActor))
  (Utilities|IsValid _mapiconactor
    (:Is%20Valid
      (Class|BPSpireMinimapActor|SetMapLegendText _mapiconactor New%20Text))))

(fn SetMinimapIcon
  (params (NewIcon :type Object%20Reference))
  (bind _mapiconactor (Variables|Minimap|GetMapIconActor))
  (Utilities|IsValid _mapiconactor
    (:Is%20Valid
      (Default|SetMapIcon _mapiconactor NewIcon))))

(fn SpawnMinimapIcon
  ;[0 0 0] Create the MapIconActor. This is how Heist Bases/God Challenges/NitroDrome handled the icon to avoid relevancy issues.
  (bind _returnvalue (Game|SpawnActorfromClass "/SpiresGameplay/Gameplay/Actors/BP_Spire_Minimap_Actor.BP_Spire_Minimap_Actor_C" (Math|Transform|MakeTransform (Transformation|GetActorLocation (Components|GetOwner)) (Transformation|GetActorRotation (Components|GetOwner))) "AlwaysSpawn" "MultiplyWithRoot" 0 0))
  (Variables|Minimap|SetMapIconActor _returnvalue)
  (CallFunction|SetMinimapIcon (Variables|Minimap|GetMinimap%20Icon%20Inactive))
  (CallFunction|SetMapLegendIconText (Variables|Default|GetMapIconText_Inactive)))

(fn WaitUntilPlayspaceIsAvailable
  (bind _returnvalue (SpatialGameplay|GetPlayspaceforLocation self (Transformation|GetActorLocation (Components|GetOwner))))
  (bind _asbp_spire_playspace (Utilities|Casting|CastToBP_SpirePlayspace _returnvalue)
    (:then
      (CallFunction|OnSpirePlayspaceAvailable _asbp_spire_playspace))
    (:CastFailed
      (bind _returnvalue_2 (Utilities|Time|SetTimerbyEvent (EventDispatchers|CreateEvent :!event WaitUntilPlayspaceIsAvailable) 0.5 false true))
      (Variables|Default|SetTryGetPlayspaceTimerHandle _returnvalue_2))))
